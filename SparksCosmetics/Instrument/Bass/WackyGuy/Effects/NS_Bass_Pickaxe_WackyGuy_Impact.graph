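# Niagara impact effect for the WackyGuy bass pickaxe, built from one-shot sprite bursts for core, flare, colored streaks and sparks, plus a light and mesh debris.
# System-level DirectionVector steers most emitters, and TOD Color tints the core and light.
# FN_PhysicalSurface uses User.PhysicalSurface to set debris velocity and cone angle per material.
# Spawn-group masks vary particles within emitters, while front-end modifiers and User.FrontEndScale adjust size, speed and brightness for previews.
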NiagaraSystem /SparksCosmetics/Instrument/Bass/WackyGuy/Effects/NS_Bass_Pickaxe_WackyGuy_Impact
dependencies {
 usedDataInterfaces [
  &NiagaraDataInterfaceArrayColor
  &NiagaraDataInterfaceVector2DCurve
  &NiagaraDataInterfaceCurve
  &NiagaraDataInterfaceColorCurve
  &NiagaraDataInterfaceVectorCurve
  &NiagaraDataInterfaceArrayFloat3
 ]
 usedDynamicInputs [
  &SelectLinearColorFromArray
  &ReturnExecIndex
  /Niagara/DynamicInputs/Math/Clamp_Vector
  &SelectVectorFromArray
 ]
 usedModules [
  &SystemState
  &self:SystemSpawnScript.SystemScriptSource.SystemScriptGraph.NiagaraNodeAssignment_0.SetVariables_#1
  &self:SystemSpawnScript.SystemScriptSource.SystemScriptGraph.NiagaraNodeAssignment_1.SetVariables_#2
 ]
}
emitters {
 Core_Impact_Burst {
  emitterUpdateScript [
   &EmitterState()
   &SpawnBurst_Instantaneous(Spawn Group=1, Use Spawn Probability=false, UseLoopCountLimit=false)
   &SpawnBurst_Instantaneous(Spawn Count=7, Spawn Group=4, Use Spawn Probability=true, UseLoopCountLimit=false) {module=SpawnBurst_Instantaneous004}
  ]
  particleSpawnScript [
   /SetVariables(Particles.VisibilityTag=&SpawnGroup()) {isSetParameters=true}
   &InitializeParticle(
    Lifetime=&MaskFloatBySpawnGroup(
     Masked Float=&FN_FrontEndFloatModifier(Gameplay=0.13, Front End=0.15)
     Passthrough Float=&UniformRangedFloat(Minimum=0.15, Maximum=0.2)
     Spawn Group Mask=1
    )
    Mesh Scale Mode=::NewEnumerator3
    Mesh Uniform Scale=0.35
    Sprite Rotation Mode=::NewEnumerator2
    Sprite Size=&MaskVector2dBySpawnGroup(
     Masked Vector2D=&UniformRangedVector2D(Minimum=<10 45>, Maximum=<12 65>)
     Passthrough Vector2D=&Vector2DFromFloat(Value=120)
     Spawn Group Mask=4
    )
    Sprite Size Mode=::NewEnumerator4
    UsePositionOffset=false
    Write Material Random=true
   )
   &FN_PerParticleLOD(Enable SpriteSize Distance Scaling=true, Sprite Size Scale Far=<2.5 2.5>)
   &ShapeLocation(
    MaskSpawnGroup=false
    Sphere Distribution=::NewEnumerator3
    Sphere Radius=&MaskFloatBySpawnGroup(Masked Float=3, Passthrough Float=0, Spawn Group Mask=4)
    Uniform Spiral Amount=0.8
    UseOwnerScale=true
   )
   &AddVelocityFromPoint(
    Module.Use Falloff Distance=false
    Velocity Strength=&MaskFloatBySpawnGroup(
     Masked Float=&FloatFromCurve(
      FloatCurve=DataInterface() {propertyValues {Curve {keys [<at 0 0>,<at 1 15>]}}}
      CurveIndex=&RandomRangeFloat(Minimum=0, Maximum=1)
     )
     Passthrough Float=0
     Spawn Group Mask=4
    )
   )
   &AddVelocity(
    Velocity=$System.DirectionVector
    Velocity Speed Scale=&MaskFloatBySpawnGroup(Masked Float=&RandomRangeFloat(Minimum=25, Maximum=150), Passthrough Float=0, Spawn Group Mask=4)
   )
   /SetVariables(
    Particles.FlareCardColor:LinearColor=$System.TOD Color
    Particles.FlareCardDynamicParameter:Vector4f=&MakeVector4(X=0.2, Y=1, Z=0, W=0)
   ) {isSetParameters=true}
   &CameraOffset(Camera Offset Amount=15)
   &ApplyInitialForces()
  ]
  particleUpdateScript [
   &ParticleState()
   &ScaleSpriteSize(
    Scale Factor=&MaskVector2dBySpawnGroup(
     Masked Vector2D=&Vector2DFromFloat(Value=&FloatFromCurve(
       FloatCurve=DataInterface() {propertyValues {Curve {
          defaultValue=1
          keys [
           {interpMode=RCIM_Cubic, tangentMode=RCTM_User, time=0, value=1}
           {arriveTangent=-3.0339, interpMode=RCIM_Cubic, leaveTangent=-3.0339, tangentMode=RCTM_User, time=1, value=0}
          ]
         }}}
       Scale Curve=&FN_FrontEndFloatModifier(Gameplay=1, Front End=$User.FrontEndScale)
      ))
     Passthrough Vector2D=&MakeVector2D(
      X=&FloatFromCurve(
       FloatCurve=DataInterface() {propertyValues {Curve {
          defaultValue=1
          keys [
           {arriveTangent=1.7844, interpMode=RCIM_Cubic, leaveTangent=1.7844, tangentMode=RCTM_User, time=0, value=0}
           {arriveTangentWeight=0.3771, interpMode=RCIM_Cubic, leaveTangentWeight=0.0943, tangentMode=RCTM_User, time=0.8, value=1}
           {arriveTangent=-5.8206, interpMode=RCIM_Cubic, leaveTangent=-5.8206, tangentMode=RCTM_User, time=1.0063, value=0.1176}
          ]
         }}}
       Scale Curve=&FN_FrontEndFloatModifier(Gameplay=1, Front End=$User.FrontEndScale)
      )
      Y=&FloatFromCurve(
       FloatCurve=DataInterface() {propertyValues {Curve {
          defaultValue=1
          keys [
           {arriveTangent=0.8542, interpMode=RCIM_Cubic, leaveTangent=0.8542, tangentMode=RCTM_User, time=0, value=0.3}
           {arriveTangent=3.2667, interpMode=RCIM_Cubic, leaveTangent=3.2667, tangentMode=RCTM_User, time=1, value=1.85}
          ]
         }}}
       Scale Curve=&FN_FrontEndFloatModifier(Gameplay=1, Front End=$User.FrontEndScale)
      )
     )
     Spawn Group Mask=1
    )
    Scale Sprite Size Mode=::NewEnumerator2
   )
   &LinearForce(Force=&Multiply_VectorByFloat(Vector=$System.DirectionVector, Float=&RandomRangeFloat(Minimum=25, Maximum=4500)))
   &ScaleColor(
    Scale Alpha=&FloatFromCurve(
     FloatCurve=DataInterface() {propertyValues {Curve {
        defaultValue=1
        keys [
         {arriveTangent=-1.5437, interpMode=RCIM_Cubic, leaveTangent=-1.5437, tangentMode=RCTM_User, time=0, value=1}
         {interpMode=RCIM_Cubic, tangentMode=RCTM_User, time=1, value=0}
        ]
       }}}
     Scale Curve=1.1
    )
    Scale RGB=&VectorFromFloat(Value=1)
    ScaleA=true
    ScaleRGB=true
   )
   /SetVariables(
    Particles.UpdateFlareCardColor:LinearColor=&MakeLinearColorFromVectorAndFloat(
     "Vector (RGB)"=&Multiply_Vector(
      A=&Multiply_Vector(
       A=&MakeVectorFromLinearColorRGB(Color=$Particles.Initial.FlareCardColor)
       B=&MakeVectorFromLinearColorRGB(Color=&ColorFromCurve(ColorCurve=DataInterface() {propertyValues {
           AlphaCurve {keys [<at 0 1 RCIM_Cubic>,<at 1 0 RCIM_Cubic>]}
           BlueCurve {keys [<at 0 1>,<at 1 0.1596>]}
           GreenCurve {keys [<at 0 0.5823>,<at 1 0.0006>]}
           RedCurve {keys [<at 0 0.7143>,<at 1 0.0686>]}
          }}))
      )
      B=&VectorFromFloat(Value=&FN_FrontEndFloatModifier(Gameplay=1.5, Front End=1))
     )
     "Float (Alpha)"=&FloatFromCurve(
      FloatCurve=DataInterface() {propertyValues {Curve {keys [{arriveTangent=-1.9727, interpMode=RCIM_Cubic, time=0, value=1}, <at 1 0 RCIM_Cubic>]}}}
      Scale Curve=&MakeCustomFloatFromBool(Bool=$User.bIsFrontEndPreview, True Float=0.5, False Float=1)
     )
    )
    Particles.UpdateFlareCardScale:Vector2f=&Multiply_Vector2D(
     A=$Particles.SpriteSize
     B=&MakeVector2D(
      X=&FloatFromCurve(FloatCurve=DataInterface() {propertyValues {Curve {keys [
           {interpMode=RCIM_Cubic, leaveTangentWeight=0.4515, tangentMode=RCTM_User, time=0, value=1}
           {arriveTangent=-1.7264, interpMode=RCIM_Cubic, leaveTangent=-1.7264, tangentMode=RCTM_User, time=1, value=0}
          ]}}})
      Y=1
     )
    )
   ) {isSetParameters=true}
   /SetVariables(Particles.ColorDirectionalSpikes:LinearColor=&MultiplyLinearColor(
     Linear Color=&ColorFromCurve(ColorCurve=DataInterface() {propertyValues {
        AlphaCurve {keys [<at 0.2866 1>,<at 1 0>]}
        BlueCurve {keys [<at 0.0768 2>,<at 0.379 0.7034>,<at 0.8274 0.0021>]}
        GreenCurve {keys [<at 0.0768 2>,<at 0.379 0.6908>,<at 0.8274 0.0028>]}
        RedCurve {keys [<at 0.0768 2>,<at 0.379 1.5>,<at 0.8274 0.02>]}
       }})
     Scale Factor=&MakeVector4_FromLinearColor(LinearColor=$System.TOD Color)
    )) {isSetParameters=true}
   &Drag(
    Drag=&MaskFloatBySpawnGroup(Masked Float=1, Passthrough Float=2, Spawn Group Mask=4)
    Use Linear Drag=true
    Use Rotational Drag=true
   )
   &SolveForcesAndVelocity(Clamp Velocity=false, Limit Acceleration=false)
   &FN_Light_Attributes(
    Color=$System.TOD Color
    Create Color Attribute=true
    Create Exponent Attribute=true
    Create Position Attribute=true
    Create Radius Attribute=true
    Exponent=4
    Position=<0 5 0>
    Radius=90
    Scale Color=&FloatFromCurve(
     FloatCurve=DataInterface() {propertyValues {Curve {keys [
         {interpMode=RCIM_Cubic, tangentMode=RCTM_User, time=0, value=1}
         {arriveTangent=-0.5426, interpMode=RCIM_Cubic, leaveTangent=-0.5426, tangentMode=RCTM_User, time=0.8, value=0}
        ]}}}
     Scale Curve=15
    )
   )
  ]
  rendererClasses [&NiagaraSpriteRendererProperties, &NiagaraLightRendererProperties]
  renderers [
   {rendererClass=&NiagaraSpriteRendererProperties}
   {rendererClass=&NiagaraLightRendererProperties, rendererIndex=1}
   {rendererClass=&NiagaraSpriteRendererProperties, rendererIndex=2}
   {rendererClass=&NiagaraSpriteRendererProperties, rendererIndex=3}
  ]
 }
 DirectionalSparks {
  emitterUpdateScript [
   &EmitterState()
   &SpawnBurst_Instantaneous(Spawn Count=50, Use Spawn Probability=false, UseLoopCountLimit=false)
  ]
  particleSpawnScript [
   /SetVariables(Particles.Energy:NiagaraFloat=&RandomRangeFloat(Minimum=0, Maximum=1)) {isSetParameters=true}
   &InitializeParticle(
    Color Maximum=<rgba 2 1.3374 0.0123 1>
    Color Minimum=<rgba 1.5 0.1334 0.0092 1>
    Color Mode=::NewEnumerator2
    Lifetime=&FloatFromCurve(
     FloatCurve=DataInterface() {propertyValues {Curve {keys [
         {arriveTangent=0.1901, interpMode=RCIM_Cubic, leaveTangent=0.1901, tangentMode=RCTM_User, time=0, value=0.1}
         {arriveTangent=0.5438, interpMode=RCIM_Cubic, leaveTangent=0.5438, tangentMode=RCTM_User, time=0.8, value=0.25}
         {arriveTangent=3.2278, interpMode=RCIM_Cubic, leaveTangent=3.2278, tangentMode=RCTM_User, time=1, value=0.6}
        ]}}}
     CurveIndex=$Particles.Energy
    )
    Sprite Size=&Multiply_Vector2D_ByFloat(
     Float=&Multiply_Float(
      A=&FN_FrontEndFloatModifier(Gameplay=&RandomRangeFloat(Minimum=0.5, Maximum=1), Front End=$User.FrontEndScale)
      B=&RandomRangeFloat(Minimum=9, Maximum=10)
     )
     Vector2D=&FN_FrontEndVector2Modifier(Front End:Vector2f=<0.5 1>, Gameplay:Vector2f=<1 1>)
    )
    Sprite Size Mode=::NewEnumerator4
    UsePositionOffset=false
    Write Material Random=true
   )
   &ShapeLocation(
    Cylinder Height=50
    Cylinder Radius=5
    MaskSpawnGroup=false
    Override Seed=false
    Rotation Axis=$System.DirectionVector
    Rotation Mode=::NewEnumerator1
    Shape Primitive=::NewEnumerator1
    Surface Only=false
    UseOwnerScale=true
   )
   &CameraOffset(Camera Offset Amount=20)
   &AddVelocityFromPoint(
    Module.Use Falloff Distance=false
    Velocity Strength=&Multiply_Float(
     A=&RandomRangeFloat(Minimum=150, Maximum=200)
     B=&FN_FrontEndFloatModifier(Gameplay=1, Front End=$User.FrontEndScale)
    )
   )
   &AddVelocity(
    Velocity=$System.DirectionVector
    Velocity Speed Scale=&Multiply_Float(
     A=&Lerp_Float(A=100, B=500, Alpha=$Particles.Energy)
     B=&MakeCustomFloatFromBool(Bool=$User.bIsFrontEndPreview, True Float=$User.FrontEndScale, False Float=1)
    )
   )
  ]
  particleUpdateScript [
   &ParticleState()
   &AccelerationForce(Acceleration=&RandomRangeVector(Minimum=<0 0 -2000>, Maximum=<0 0 -500>))
   &Drag(Drag=&UniformRangedFloat(Minimum=2, Maximum=6), Use Linear Drag=true, Use Rotational Drag=true)
   &SolveForcesAndVelocity(Clamp Velocity=false, Limit Acceleration=false)
   &ScaleSpriteSizeBySpeed(Max Scale Factor=<1 5>, Min Scale Factor=<0.5 0.2>)
   &ScaleSpriteSize(
    Uniform Curve Scale=&MakeCustomFloatFromBool(Bool=$User.bIsFrontEndPreview, True Float=1, False Float=0.75)
    Uniform Curve Sprite Scale=DataInterface() {propertyValues {Curve {keys [{arriveTangent=-6.725, interpMode=RCIM_Cubic, leaveTangent=-6.725, tangentMode=RCTM_User, time=0.65, value=1}, <at 1 0 RCIM_Cubic>]}}}
   )
   &ScaleColor(
    Scale RGB=&VectorFromFloat(Value=&FloatFromCurve(FloatCurve=DataInterface() {propertyValues {Curve {keys [
          {interpMode=RCIM_Cubic, tangentMode=RCTM_User, time=0, value=10}
          <at 0.3 1>
          {interpMode=RCIM_Cubic, tangentMode=RCTM_User, time=1, value=0}
         ]}}}))
    ScaleA=true
    ScaleRGB=true
   )
   /SetVariables(
    Particles.GlowColor=&MakeLinearColorFromVectorAndFloat(
     "Vector (RGB)"=&MakeVectorFromLinearColorRGB(Color=<rgba 1 0.2237 0 1>)
     "Float (Alpha)"=&FloatFromCurve(FloatCurve=DataInterface() {propertyValues {Curve {keys [<at 0.5 1>,<at 1 0>]}}}, Scale Curve=0.1)
    )
    Particles.GlowSize=&Multiply_Vector2D(A=$Particles.SpriteSize, B=<8 2>)
   ) {enabled=false, isSetParameters=true}
  ]
  rendererClasses [&NiagaraSpriteRendererProperties]
  renderers [
   {rendererClass=&NiagaraSpriteRendererProperties}
   {rendererClass=&NiagaraSpriteRendererProperties, rendererIndex=1}
   {rendererClass=&NiagaraSpriteRendererProperties, rendererIndex=2}
  ]
 }
 DirectionalStreaks {
  emitterUpdateScript [
   &EmitterState()
   &SpawnBurst_Instantaneous(Spawn Count=3, Use Spawn Probability=false, UseLoopCountLimit=false)
  ]
  particleSpawnScript [
   &InitializeParticle(
    Color Maximum=<rgba 1 0.1654 0.0226 1>
    Color Minimum=<rgba 1 0.0365 0.0868 1>
    Color Mode=::NewEnumerator2
    Lifetime Max=0.15
    Lifetime Min=0.1
    Lifetime Mode=::NewEnumerator1
    Sprite Size=&Multiply_Vector2D_ByFloat(
     Float=&FN_FrontEndFloatModifier(Gameplay=1, Front End=$User.FrontEndScale)
     Vector2D=&UniformRangedVector2D(Minimum=<20 32>, Maximum=<25 56>)
    )
    Sprite Size Mode=::NewEnumerator4
    Sprite UV Mode=::NewEnumerator1
    UsePositionOffset=false
    Write Material Random=true
   ) {module=InitializeParticle001}
   &AddVelocity(
    Module.Use Falloff Distance=false
    Override Seed=false
    Velocity Mode=::NewEnumerator1
    Velocity Speed=&Multiply_Float(A=20, B=&FN_FrontEndFloatModifier(Gameplay=1, Front End=$User.FrontEndScale))
   ) {enabled=false, module=AddVelocity002}
   &SubUVAnimation(End Frame=3, SubUV Animation Mode:ENiagara_SubUVLookupMode=::NewEnumerator2)
   &AddVelocity(
    Velocity=$System.DirectionVector
    Velocity Speed Scale=&Multiply_Float(A=100, B=&FN_FrontEndFloatModifier(Gameplay=1, Front End=$User.FrontEndScale))
   )
   &AddVelocity(
    Inner Cone Angle=30
    Module.UseConeFalloff=false
    Module.UseVelDistribution=false
    Override Seed=false
    Velocity Mode=::NewEnumerator2
    Velocity Speed=&Multiply_Float(A=100, B=&FN_FrontEndFloatModifier(Gameplay=1, Front End=$User.FrontEndScale))
   ) {module=AddVelocity001}
   &CameraOffset(Camera Offset Amount=25)
  ]
  particleUpdateScript [
   &ParticleState()
   &SolveForcesAndVelocity(Clamp Velocity=false, Limit Acceleration=false)
   &ScaleSpriteSize(Scale Factor=&Vector2DFromCurve(
     Vector2Curve=DataInterface() {propertyValues {
       XCurve {keys [<at 0 1 RCIM_Cubic>,<at 1 0 RCIM_Cubic>]}
       YCurve {keys [{arriveTangent=3.9291, interpMode=RCIM_Cubic, leaveTangent=3.9291, tangentMode=RCTM_User, time=0, value=0}, <at 1 2 RCIM_Cubic>]}
      }}
     ScaleCurve=&Multiply_Vector2D_ByFloat(
      Float=1
      Vector2D=&MakeCustomVector2DFromBool(Bool=$User.bIsFrontEndPreview, True Vector2D=<1.5 1>, False Vector2D=<1 1>)
     )
    ))
   &ScaleColor(
    Scale Alpha=&FloatFromCurve(
     FloatCurve=DataInterface() {propertyValues {Curve {keys [<at 0 1 RCIM_Cubic>,<at 1 0>]}}}
     Scale Curve=&FNTimeOfDayScalar(Night Scale=0.5, Morning Scale=1, Evening Scale=0.5)
    )
    Scale RGB=&VectorFromFloat(Value=&FloatFromCurve(
      FloatCurve=DataInterface() {propertyValues {Curve {keys [
          {arriveTangent=-11.1455, interpMode=RCIM_Cubic, leaveTangent=-11.1455, tangentMode=RCTM_User, time=0, value=5}
          {arriveTangent=-0.8731, interpMode=RCIM_Cubic, leaveTangent=-0.8731, tangentMode=RCTM_User, time=1, value=1}
         ]}}}
      Scale Curve=&MakeCustomFloatFromBool(Bool=$User.bIsFrontEndPreview, True Float=0.2, False Float=1)
     ))
    ScaleA=true
    ScaleRGB=true
   )
  ]
  rendererClasses [&NiagaraSpriteRendererProperties]
  renderers [{rendererClass=&NiagaraSpriteRendererProperties}]
 }
 Directional_FN__PhysMatDebris {
  emitterUpdateScript [
   &EmitterState()
   &SpawnBurst_Instantaneous(
    Spawn Count=&RandomRangeInt(Minimum=5, Maximum=10)
    Use Spawn Probability=false
    UseLoopCountLimit=false
   )
  ]
  particleSpawnScript [
   /SetVariables(Particles.BiasedRandom=&FloatFromCurve(
     FloatCurve=DataInterface() {propertyValues {Curve {keys [
         <at 0.2 1 RCIM_Cubic>
         {arriveTangent=-3.3394, interpMode=RCIM_Cubic, leaveTangent=-3.3394, tangentMode=RCTM_User, time=0.5, value=0.5}
         <at 0.8 0 RCIM_Cubic>
        ]}}}
     CurveIndex=&RandomRangeFloat(Minimum=0, Maximum=1)
    )) {isSetParameters=true}
   &InitializeParticle(
    Color Mode=::NewEnumerator0
    Lifetime=&RandomRangeFloat(Minimum=0.5, Maximum=1.25)
    Mass=&Lerp_Float(A=0.6, B=1.2, Alpha=$Particles.BiasedRandom)
    Mass Mode=::NewEnumerator1
    Mesh Scale=&Multiply_VectorByFloat(
     Vector=&RandomRangeVector(Minimum=<0.9 0.4 0.5>, Maximum=<1 1 1>)
     Float=&Lerp_Float(A=0.7, B=1, Alpha=$Particles.BiasedRandom)
    )
    Mesh Scale Mode=::NewEnumerator4
    UsePositionOffset=false
    Write Material Random=true
   )
   &FN_PhysicalSurface(
    Dirt Mesh Count=2
    Glass Mesh Count=3
    Ice Mesh Count=2
    Metal Gravity=&RandomRangeFloat(Minimum=-350, Maximum=-1800)
    Metal Mesh Count=2
    Metal Velocity=&RandomRangeFloat(Minimum=80, Maximum=375)
    Metal Velocity Cone Angle=130
    Physical Surface=$User.PhysicalSurface
    Stone Gravity=&RandomRangeFloat(Minimum=-700, Maximum=-1250)
    Stone Mesh Count=2
    Stone Velocity=&RandomRangeFloat(Minimum=80, Maximum=550)
    Stone Velocity Cone Angle=120
    Wood Gravity=&RandomRangeFloat(Minimum=-450, Maximum=-900)
    Wood Mesh Count=2
    Wood Scale=1.5
    Wood Velocity=&RandomRangeFloat(Minimum=80, Maximum=600)
    Wood Velocity Cone Angle=130
   )
   &AddVelocityInCone(
    Cone Angle=$Output.FN_PhysicalSurface.SurfaceResponseConeAngle
    Cone Axis=$System.DirectionVector
    Module.Use Velocity Falloff On Cone Axis=false
    Velocity Strength=$Output.FN_PhysicalSurface.SurfaceResponseVelocity
   )
   &InitialMeshOrientation(Use Rotation Vector=false)
   &AddRotationalVelocity(Rotation Rate=&RandomVector(Vector Scale=&RandomRangeFloat(Minimum=1, Maximum=2)))
   &ApplyInitialForces()
   /SetVariables(Particles.GravityOverrideValue=$Output.FN_PhysicalSurface.GravityOverride) {enabled=false, isSetParameters=true}
  ]
  particleUpdateScript [
   &ParticleState()
   &AccelerationForce(Acceleration=&RandomRangeVector(Minimum=<0 0 -2000>, Maximum=<0 0 -500>))
   &Drag(
    Drag=&MaskFloatBySpawnGroup(
     Masked Float=&RandomRangeFloat(Minimum=1, Maximum=4)
     Passthrough Float=&UniformRangedFloat(Minimum=2, Maximum=6)
     Spawn Group Mask=1
    )
    Use Linear Drag=true
    Use Rotational Drag=true
   )
   &GravityForce(Gravity=&RandomRangeVector(Minimum=<0 0 -500>, Maximum=<0 0 -2000>)) {enabled=false}
   &Collision(
    Advanced Aging Rate=1
    Age Colliding Particles=false
    CPU Collision Trace Channel=ECC_WorldStatic
    Collision Enabled=&Boolean_AndOperation(
     Bool1=&SetBoolByFloatComparison(A=$Particles.Initial.Lifetime, B=1)
     Bool2=&SetBoolByFloatComparison(A=$Particles.Age, B=0.12)
    )
    Control Roll On Collision=true
    Friction=1
    Friction During a Bounce=0.25
    Max CPU Trace Length=100
    Module.Control Rest State Via Surface Normal=false
    Module.EnableMaxCollisionCount=false
    Module.Randomize Collision Normal Vector=true
    Randomize Collision Normal=0.25
    ScaleRateOfRotation=false
   ) {enabled=false}
   &SolveForcesAndVelocity(Clamp Velocity=false, Limit Acceleration=false)
   &ScaleMeshSize(Scale Factor=&VectorFromFloat(Value=&FloatFromCurve(FloatCurve=DataInterface() {propertyValues {Curve {keys [<at 0.6 1 RCIM_Cubic>, {arriveTangent=-6.308, interpMode=RCIM_Cubic, leaveTangent=-6.308, tangentMode=RCTM_User, time=1, value=0}]}}})))
   &DynamicMaterialParameters(Module.Param0WriteEnabled=true, Module.Param1WriteEnabled=true, Module.Param2WriteEnabled=false, Module.Param3WriteEnabled=false)
  ]
  rendererClasses [&NiagaraSpriteRendererProperties, &NiagaraMeshRendererProperties]
  renderers [
   {rendererClass=&NiagaraSpriteRendererProperties}
   {rendererClass=&NiagaraMeshRendererProperties, rendererIndex=1}
   {rendererClass=&NiagaraMeshRendererProperties, rendererIndex=2}
   {rendererClass=&NiagaraMeshRendererProperties, rendererIndex=3}
   {rendererClass=&NiagaraMeshRendererProperties, rendererIndex=4}
   {rendererClass=&NiagaraMeshRendererProperties, rendererIndex=5}
   {rendererClass=&NiagaraMeshRendererProperties, rendererIndex=6}
   {rendererClass=&NiagaraMeshRendererProperties, rendererIndex=7}
   {rendererClass=&NiagaraMeshRendererProperties, rendererIndex=8}
  ]
 }
 Flash_Flare {
  emitterUpdateScript [
   &EmitterState()
   &SpawnBurst_Instantaneous(Use Spawn Probability=false, UseLoopCountLimit=false)
   &SpawnBurst_Instantaneous(Spawn Group=1, Use Spawn Probability=false, UseLoopCountLimit=false) {module=SpawnBurst_Instantaneous001}
  ]
  particleSpawnScript [
   &InitializeParticle(
    Color=<rgba 2 2 2 1>
    Lifetime=0.18
    Sprite Size=&MaskVector2dBySpawnGroup(Masked Vector2D=<80 100>, Passthrough Vector2D=<80 240>, Spawn Group Mask=0)
    Sprite Size Mode=::NewEnumerator4
    UsePositionOffset=false
    Write Material Random=true
   )
   &CameraOffset(Camera Offset Amount=25)
   &DynamicMaterialParameters(
    Index 0 Param 1=&MaskFloatBySpawnGroup(Masked Float=0.2, Passthrough Float=0)
    Index 0 Param 2=&MaskFloatBySpawnGroup(Masked Float=1, Passthrough Float=0)
    Index 3 Param 1=1.5
    Index 3 Param 2=1
    Module.Param0WriteEnabled=true
    Module.Param0WriteEnabled3=true
    Module.Param1WriteEnabled=true
    Module.Param1WriteEnabled3=true
    Module.Param2WriteEnabled=true
    Module.Param2WriteEnabled3=true
    Module.Param3WriteEnabled=true
    Module.Param3WriteEnabled3=true
    Write Parameter Index 3=true
   )
   &AddVelocity(
    Velocity=$System.DirectionVector
    Velocity Speed Scale=&Multiply_Float(
     A=&RandomRangeFloat(Minimum=200, Maximum=300)
     B=&MaskFloatBySpawnGroup(Masked Float=0.5, Passthrough Float=1)
    )
   )
  ]
  particleUpdateScript [
   &ParticleState()
   &ScaleSpriteSize(
    Non-Uniform Curve Sprite Scale=DataInterface() {propertyValues {
      XCurve {keys [
        {arriveTangent=20.4873, interpMode=RCIM_Cubic, leaveTangent=20.4873, tangentMode=RCTM_User, time=0, value=0}
        {arriveTangent=-0.1, interpMode=RCIM_Cubic, leaveTangent=-0.1, time=0.4261, value=0.9194}
        {arriveTangent=-0.3934, interpMode=RCIM_Cubic, leaveTangent=-0.3934, tangentMode=RCTM_User, time=1, value=-0.1}
       ]}
      YCurve {keys [
        {arriveTangent=19.7829, interpMode=RCIM_Cubic, leaveTangent=19.7829, tangentMode=RCTM_User, time=0, value=0}
        {arriveTangentWeight=0.0333, interpMode=RCIM_Cubic, leaveTangentWeight=0.3, tangentMode=RCTM_User, time=0.1, value=1.025}
        {arriveTangent=-0.0827, interpMode=RCIM_Cubic, leaveTangent=-0.0827, tangentMode=RCTM_User, time=1, value=0.8}
       ]}
     }}
    Scale Sprite Size Mode=::NewEnumerator3
   )
   &SolveForcesAndVelocity(Clamp Velocity=false, Limit Acceleration=false)
   &ScaleColor(
    Scale Mode=1
    Scale RGBA=/Niagara/DynamicInputs/Helpers/MaskVector4(
     Mask Source:NiagaraInt32=0
     Mask Parameter:NiagaraInt32=?
     Mask:NiagaraInt32=0
     Masked Vector4:Vector4f=&Vector4FromCurve(Vector4Curve=DataInterface() {propertyValues {
        AlphaCurve {keys [<at 0 1>]}
        BlueCurve {keys [<at 0.155 0>,<at 0.4795 0>,<at 0.799 0.298>,<at 1 1>]}
        GreenCurve {keys [<at 0.155 0.5843>,<at 0.4795 0.1461>,<at 0.799 0.1009>,<at 1 0>]}
        RedCurve {keys [<at 0.155 1>,<at 0.4795 1>,<at 0.799 1>,<at 1 0.2667>]}
       }})
     Passthrough Vector4:Vector4f=&Vector4FromCurve(Vector4Curve=DataInterface() {propertyValues {
        AlphaCurve {keys [<at 0 1>]}
        BlueCurve {keys [<at 0 0>,<at 0.1484 0>,<at 0.5184 0.298>,<at 0.9136 1>]}
        GreenCurve {keys [<at 0 0.5843>,<at 0.1484 0.1461>,<at 0.5184 0.1009>,<at 0.9136 0>]}
        RedCurve {keys [<at 0 1>,<at 0.1484 1>,<at 0.5184 1>,<at 0.9136 0.2667>]}
       }})
    )
    ScaleRGBA=true
   )
  ]
  rendererClasses [&NiagaraSpriteRendererProperties]
  renderers [{rendererClass=&NiagaraSpriteRendererProperties}]
 }
 Streaks {
  emitterUpdateScript [
   &EmitterState()
   &SpawnBurst_Instantaneous(Spawn Count=3, Use Spawn Probability=false, UseLoopCountLimit=false)
   &SpawnBurst_Instantaneous(Spawn Count=3, Spawn Group=1, Use Spawn Probability=true, UseLoopCountLimit=false) {module=SpawnBurst_Instantaneous001}
  ]
  particleSpawnScript [
   &InitializeParticle(
    Color Maximum=&MaskLinearColorBySpawnGroup(Masked Linear Color=<rgba 1 0.05 0.5754 0.1>, Passthrough Linear Color=<rgba 0.1 0.2057 2 0.1>)
    Color Minimum=<rgba 0.2242 0.05 1 0.1>
    Color Mode=::NewEnumerator2
    Lifetime Max=0.15
    Lifetime Min=0.1
    Lifetime Mode=::NewEnumerator1
    Sprite Size=&Multiply_Vector2D_ByFloat(
     Float=&FN_FrontEndFloatModifier(Gameplay=1, Front End=0.5)
     Vector2D=&UniformRangedVector2D(Minimum=<20 40>, Maximum=<25 70>)
    )
    Sprite Size Mode=::NewEnumerator4
    Sprite UV Mode=::NewEnumerator1
    UsePositionOffset=false
    Write Material Random=true
   ) {module=InitializeParticle001}
   &AddVelocity(
    Module.Use Falloff Distance=false
    Override Seed=false
    Velocity Mode=::NewEnumerator1
    Velocity Speed=&Multiply_Float(A=25, B=&FN_FrontEndFloatModifier(Gameplay=1, Front End=$User.FrontEndScale))
   ) {module=AddVelocity001}
   &SubUVAnimation(End Frame=3, SubUV Animation Mode:ENiagara_SubUVLookupMode=::NewEnumerator2)
   &AddVelocity(
    Velocity=$System.DirectionVector
    Velocity Speed Scale=&MaskFloatBySpawnGroup(Masked Float=100, Passthrough Float=50)
   )
   &CameraOffset(Camera Offset Amount=16)
  ]
  particleUpdateScript [
   &ParticleState()
   &SolveForcesAndVelocity(Clamp Velocity=false, Limit Acceleration=false)
   &ScaleSpriteSize(Scale Factor=&Vector2DFromCurve(
     Vector2Curve=DataInterface() {propertyValues {
       XCurve {keys [
         <at 0 1 RCIM_Cubic>
         {arriveTangent=-1, interpMode=RCIM_Cubic, leaveTangent=-1, time=0.5, value=0.5}
         {arriveTangent=-1.9017, interpMode=RCIM_Cubic, leaveTangent=-1.9017, tangentMode=RCTM_User, time=1, value=0}
        ]}
       YCurve {keys [{arriveTangent=3.9291, interpMode=RCIM_Cubic, leaveTangent=3.9291, tangentMode=RCTM_User, time=0, value=0}, <at 1 2 RCIM_Cubic>]}
      }}
     ScaleCurve=&MaskVector2dBySpawnGroup(Masked Vector2D=<1.5 1>, Passthrough Vector2D=<1.5 0.8>, Spawn Group Mask=0)
    ))
   &ScaleColor(
    Scale Alpha=&FloatFromCurve(
     FloatCurve=DataInterface() {propertyValues {Curve {keys [<at 0 1 RCIM_Cubic>, {arriveTangent=-3.2219, tangentMode=RCTM_User, time=1, value=0}]}}}
     Scale Curve=&FNTimeOfDayScalar(Night Scale=0.5, Morning Scale=1, Evening Scale=0.5)
    )
    Scale RGB=&Multiply_VectorByFloat(
     Vector=&VectorFromCurve(VectorCurve=DataInterface() {propertyValues {XCurve {keys [<at 0 1>,<at 1 0.75>]}, YCurve {keys [<at 0 1>,<at 1 0.5>]}, ZCurve {keys [<at 0 1>,<at 1 1>]}}})
     Float=&FloatFromCurve(FloatCurve=DataInterface() {propertyValues {Curve {keys [
          {arriveTangent=-32.1379, interpMode=RCIM_Cubic, leaveTangent=-32.1379, tangentMode=RCTM_User, time=0, value=10}
          {arriveTangent=-9, interpMode=RCIM_Cubic, leaveTangent=-9, time=0.2879, value=4.4753}
          {arriveTangent=-7.996, interpMode=RCIM_Cubic, leaveTangent=-7.996, tangentMode=RCTM_User, time=1, value=1}
         ]}}})
    )
    ScaleA=true
    ScaleRGB=true
   )
  ]
  rendererClasses [&NiagaraSpriteRendererProperties]
  renderers [{rendererClass=&NiagaraSpriteRendererProperties}]
 }
}
summary {
 systemName=NS_Bass_Pickaxe_WackyGuy_Impact
 emitters [
  {emitterName=DirectionalStreaks, rendererClasses [&NiagaraSpriteRendererProperties]}
  {
   emitterName=Directional_FN__PhysMatDebris
   rendererClasses [&NiagaraSpriteRendererProperties, &NiagaraMeshRendererProperties]
  }
  {emitterName=DirectionalSparks, rendererClasses [&NiagaraSpriteRendererProperties]}
  {emitterName=Flash_Flare, rendererClasses [&NiagaraSpriteRendererProperties]}
  {
   emitterName=Core_Impact_Burst
   rendererClasses [&NiagaraSpriteRendererProperties, &NiagaraLightRendererProperties]
  }
  {emitterName=Streaks, rendererClasses [&NiagaraSpriteRendererProperties]}
 ]
 userVariables [
  User.bIsFrontEnd {type=NiagaraBool}
  User.bIsFrontEndPreview {type=NiagaraBool}
  User.FrontEndScale {type=NiagaraFloat, defaultValue=0.6, description=""}
  User.PhysicalSurface {type=NiagaraInt32}
  User.SwingSectionIndex {type=NiagaraInt32}
 ]
}
